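# Unlit translucent two-sided material for the box light effect in the GateHound player emote, usable on Niagara sprite and mesh particles.
# A vertical two-color gradient is boosted, tinted by particle color and passed through an emissive adjustment function.
# Opacity multiplies panning cloud noise, a V-based fade with optional sine edge fade, double-sided Fresnel and particle alpha.
# A shading path switch applies depth fade on all but the mobile path.
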Material /Game/Animation/Game/MainPlayer/Emotes/GateHound/FX/Materials/M_Gatehound_Emote_BoxLight
graph {
 properties {
  bUsedWithNiagaraMeshParticles=true
  bUsedWithNiagaraSprites=true
  blendMode=BLEND_Translucent
  pixelDepthOffsetMode=PDOM_Legacy
  shadingModel=MSM_Unlit
  twoSided=true
  usedShadingModels=Unlit
 }
 nodes [
  n1 = TextureCoordinate()
  n2 = LinearInterpolate(
   A=VectorParameter() {parameterName=ColorB, defaultValue=<rgba 1 0.1667 0 1>}.RGB
   B=VectorParameter() {parameterName=ColorA, defaultValue=<rgba 1 0.5 0 1>}.RGB
   Alpha=Power(Base=ComponentMask(None=@n1) {g=true}) {constExponent=2.2}
  )
  n3 = ParticleColor()
  MaterialFunctionCall(Emissive=Multiply(A=Multiply(A=@n2) {constB=10}, B=@n3.RGB)) {materialFunction=&MF_FX_EmissiveAdjustments}
  n4 = TextureCoordinate()
  n5 = AppendVector(A=ScalarParameter() {defaultValue=1, parameterName=Noise_Tile_X}, B=ScalarParameter() {parameterName=Noise_Tile_Y})
  n6 = ScalarParameter() {defaultValue=0.02, parameterName=Noise_Panner_X}
  n7 = ScalarParameter() {defaultValue=0.1, parameterName=Noise_Panner_Y}
  n8 = TextureSample(UVs=Panner(Coordinate=Multiply(A=@n4, B=@n5), Speed=AppendVector(A=@n6, B=@n7))) {texture=&AtmosphericCloudNoise01}
  n9 = ComponentMask(None=TextureCoordinate()) {g=true}
  n10 = ScalarParameter() {defaultValue=2.2, parameterName=TopFade_Power}
  n11 = Clamp(None=Multiply(A=Power(Base=@n9, Exp=@n10) {constExponent=2.2}, B=ScalarParameter() {defaultValue=5, parameterName=TopFade_Mult}) {constB=5})
  n12 = ComponentMask(None=TextureCoordinate() {uTiling=0.5}) {r=true}
  n13 = LinearInterpolate(A=@n11, B=Multiply(A=Sine(None=@n12), B=@n11), Alpha=ScalarParameter() {parameterName=Use_EdgeFade})
  n14 = ScalarParameter() {defaultValue=1, parameterName=Fresnel_Brightness}
  n15 = ScalarParameter() {defaultValue=7, parameterName=Fresnel_Exp}
  n16 = Multiply(
   A=@n3.A
   B=Multiply(
    A=Multiply(A=@n8.RGB, B=@n13)
    B=MaterialFunctionCall(Brightness=@n14, Exponent=@n15) {materialFunction=&MF_Fresnel_FX}.Fresnel Double Sided
   )
  )
  n17 = DepthFade(Opacity=@n16) {fadeDistanceDefault=5}
  ShadingPathSwitch(Default=@n17, Deferred=@n17, Mobile=@n16) {desc=Migrated from FeatureLevelSwitch}
 ]
}
