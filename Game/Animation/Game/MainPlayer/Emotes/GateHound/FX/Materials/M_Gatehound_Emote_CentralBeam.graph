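# Unlit, translucent, two-sided Niagara mesh and ribbon material for the central beam of the GateHound emote.
# Panning turbulence and distorted swirl noise modulate a sine cross-section, colored through a curve atlas with scrolling emissive pulse rings added.
# Dynamic parameters drive pan time, U tiling, erosion and grow-out, while end masks soften opacity.
# Forward and mobile paths substitute constants for the noise and skip depth fade.
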Material /Game/Animation/Game/MainPlayer/Emotes/GateHound/FX/Materials/M_Gatehound_Emote_CentralBeam
graph {
 properties {
  bAutomaticallySetUsageInEditor=false
  bEnableResponsiveAA=true
  bUsedWithNiagaraMeshParticles=true
  bUsedWithNiagaraRibbons=true
  bUsedWithStaticMesh=false
  blendMode=BLEND_Translucent
  pixelDepthOffsetMode=PDOM_Legacy
  shadingModel=MSM_Unlit
  twoSided=true
  usedShadingModels=Unlit
 }
 nodes [
  n1 = OneMinus(None=DynamicParameter() {paramNames [Time, U_Tiling_Multiply, ErodeAll, GrowOut]}.ErodeAll)
  LinearInterpolate(Alpha=@n1) {constB=0.5}
  MaterialFunctionCall() {materialFunction=&DebugTimeSine}
  n2 = TextureCoordinate()
  n3 = Add(
   A=Multiply(
    A=MaterialFunctionCall(Float2=@n2) {materialFunction=&BreakOutFloat2Components}.R
    B=ScalarParameter() {defaultValue=0.5, parameterName=PulseTile}
   )
   B=Multiply(A=NamedRerouteUsage()) {constB=-1}
  )
  n4 = CurveAtlasRowParameter(CurveTime=Frac(None=Multiply(A=@n3) {constB=2})) {
   atlas=/Game/Animation/Game/MainPlayer/Emotes/GateHound/FX/Textures/CA_Emote_GateHound
   curve=/Game/Animation/Game/MainPlayer/Emotes/GateHound/FX/Textures/Curve_Gatehound_CenterBeam
   defaultValue=1
   parameterName=RingGradient
  }
  n5 = Multiply(
   A=ScalarParameter() {defaultValue=10, parameterName=PulseEmissive}
   B=VectorParameter() {parameterName=PulseEmisColor, defaultValue=<rgba 0.8874 1.1146 1.5 1>}.RGB
  )
  n6 = MaterialFunctionCall(Float2=NamedRerouteUsage()) {materialFunction=&BreakOutFloat2Components}
  n7 = Multiply(A=Multiply(A=@n6.R, B=NamedRerouteUsage()), B=LinearInterpolate(Alpha=ParticleColor().A) {constA=1, constB=2}) {constB=0.5}
  n8 = TextureSample(UVs=Panner(Coordinate=AppendVector(A=@n7, B=@n6.G), Time=NamedRerouteUsage()) {speedX=-2.5, speedY=1}) {samplerType=SAMPLERTYPE_Grayscale, texture=&LowResBlurredNoise}
  n9 = MaterialFunctionCall(Float2=Multiply(A=NamedRerouteUsage(), B=Constant2Vector() {g=1, r=1.5})) {materialFunction=&BreakOutFloat2Components}
  n10 = AppendVector(A=Multiply(A=@n9.R, B=NamedRerouteUsage()), B=@n9.G)
  n11 = NamedRerouteUsage()
  n12 = ScalarParameter() {defaultValue=1, parameterName=Base_U}
  n13 = ScalarParameter() {defaultValue=1, parameterName=Base_V}
  n14 = TextureSample(UVs=Multiply(A=Panner(Coordinate=@n10, Time=@n11) {speedX=-2, speedY=0.5}, B=AppendVector(A=@n12, B=@n13)) {constB=0.5}) {samplerType=SAMPLERTYPE_LinearColor, texture=/Game/Effects/Fort_Effects/Effects/Environment/Desert/Dust/Organize/T_Turbulence_Seamless}
  n15 = LinearInterpolate(Alpha=ComponentMask(None=MaterialFunctionCall!(Blend=@n8.R, Base=@n14.B) {materialFunction=/Engine/Functions/Engine_MaterialFunctions03/Blends/Blend_Lighten}.Result) {r=true}) {constA=0.79, constB=1.25}
  n16 = Constant() {r=1}
  n17 = DynamicParameter() {paramNames [Time, U_Tiling_Multiply, ErodeAll, GrowOut]}
  n18 = Multiply(A=NamedRerouteUsage(), B=Constant2Vector() {g=1, r=2})
  n19 = Sine(None=Multiply(A=MaterialFunctionCall(Float2=@n18) {materialFunction=&BreakOutFloat2Components}.G) {constB=0.5})
  n20 = NamedRerouteUsage()
  n21 = MaterialFunctionCall(Float2=Multiply(A=@n20, B=Constant2Vector() {g=1, r=0.1})) {materialFunction=&BreakOutFloat2Components}
  n22 = NamedRerouteUsage()
  n23 = AppendVector(A=Power(Base=Multiply(A=@n21.R, B=@n22)) {constExponent=0.8}, B=@n21.G)
  n24 = NamedRerouteUsage()
  n25 = MaterialFunctionCall(Float2=Multiply(A=@n20, B=Constant2Vector() {g=1, r=0.2})) {materialFunction=&BreakOutFloat2Components}
  n26 = AppendVector(A=Power(Base=Multiply(A=@n25.R, B=@n22)) {constExponent=0.65}, B=@n25.G)
  n27 = MaterialFunctionCall(Float2=Multiply(A=@n20, B=Constant2Vector() {g=1, r=0.3})) {materialFunction=&BreakOutFloat2Components}
  n28 = Multiply(A=@n27.R, B=@n22)
  n29 = MaterialFunctionCall(Float2=@n20) {materialFunction=&BreakOutFloat2Components}
  n30 = Multiply(A=@n29.R, B=@n22)
  n31 = TextureSample(UVs=Panner(Coordinate=AppendVector(A=@n30, B=@n29.G), Time=NamedRerouteUsage()) {speedX=0.35}) {samplerType=SAMPLERTYPE_Normal, texture=&T_Weap_Distort_N_02}
  n32 = Add(A=AppendVector(A=Power(Base=@n28) {constExponent=0.5}, B=@n27.G), B=Divide(A=ComponentMask(None=@n31.RGB) {g=true, r=true}) {constB=8})
  n33 = MaterialFunctionCall(
   A=TextureSampleParameter2D(UVs=Panner(Coordinate=@n23, Time=@n24) {speedX=-0.5}) {parameterName=TexC, samplerType=SAMPLERTYPE_Normal, texture=/Game/Global/Textures/Default/Noise/T_SwirlNoise_N}.R
   B=TextureSampleParameter2D(UVs=Panner(Coordinate=@n26, Time=@n24) {speedX=-0.8}) {parameterName=TexD, samplerType=SAMPLERTYPE_Normal, texture=&T_SwirlNoise_N}.R
   C=TextureSampleParameter2D(
    UVs=Panner(Coordinate=@n32, Time=@n24) {speedX=-1.5, speedY=0.5}
    Apply View MipBias=TextureObject() {samplerType=SAMPLERTYPE_Normal, texture=/Game/Global/Textures/Default/Flowmaps/T_SwirlFlow_02_N}
   ) {parameterName=TexE, samplerType=SAMPLERTYPE_Normal, texture=&T_SwirlNoise_N}.R
   Alpha=@n19
  ) {materialFunction=&3ColorBlend}
  n34 = Constant() {r=2}
  n35 = Power(
   Base=Power(Base=@n19) {constExponent=2.5}
   Exp=ShadingPathSwitch(Default=LinearInterpolate(Alpha=@n33.Result) {constA=2.2, constB=5.5}, Forward=@n34, Mobile=@n34)
  )
  n36 = Multiply(
   A=Power(Base=ShadingPathSwitch(Default=@n15, Forward=@n16, Mobile=@n16), Exp=LinearInterpolate(Alpha=@n17.ErodeAll) {constA=3.5, constB=1.5}) {constExponent=1.5}
   B=@n35
  )
  n37 = CurveAtlasRowParameter(CurveTime=Saturate(None=Subtract(A=@n36, B=@n1) {constB=0})) {
   atlas=&CA_Emote_GateHound
   curve=/Game/Animation/Game/MainPlayer/Emotes/GateHound/FX/Textures/Curve_Gatehound_PurpleBeam
   parameterName=ColorCurve
  }
  n38 = Multiply(A=@n37, B=ParticleColor().RGB)
  n39 = MaterialFunctionCall(LDR Input=Saturate(None=@n38), HDR Tint=@n38) {materialFunction=&DeriveHDRfromLDR}
  n40 = Add(A=Multiply(A=Multiply(A=@n4.R, B=@n5), B=@n39.RemappedRange), B=@n39.RemappedRange)
  n41 = ScalarParameter() {defaultValue=1, parameterName=Opacity}
  MaterialFunctionCall(Emissive=Multiply(A=ScalarParameter() {defaultValue=1, parameterName=Brightness}, B=Multiply(A=@n40, B=@n41))) {materialFunction=&MF_FX_EmissiveAdjustments}
  n42 = Saturate(None=@n35)
  n43 = OneMinus(None=MaterialFunctionCall(Float2=TextureCoordinate()) {materialFunction=&BreakOutFloat2Components}.R)
  n44 = DynamicParameter() {paramNames [Time, U_Tiling_Multiply, ErodeAll, GrowOut]}
  n45 = Multiply(A=Saturate(None=Subtract(A=@n43, B=@n44.GrowOut) {constB=0})) {constB=6}
  n46 = Subtract(A=@n35, B=OneMinus(None=@n44.ErodeAll)) {constB=0}
  n47 = OneMinus(None=NamedRerouteUsage())
  n48 = Multiply(
   A=SphereMask(A=@n47, B=Constant2Vector() {g=0.5, r=0.3}) {attenuationRadius=0.7, hardnessPercent=0}
   B=ScalarParameter() {defaultValue=50, parameterName=EndSoftness}
  ) {constB=8}
  n49 = Saturate(None=Multiply(A=@n48, B=@n48))
  n50 = MaterialFunctionCall(Float2=@n47) {materialFunction=&BreakOutFloat2Components}
  n51 = Multiply(A=OneMinus(None=@n50.R), B=ScalarParameter() {defaultValue=4, parameterName=EndSoftness2}) {constA=3, constB=2}
  n52 = Multiply(A=@n50.R) {constA=3, constB=15}
  n53 = Power(Base=Multiply(A=Saturate(None=@n51), B=Saturate(None=@n52))) {constExponent=1.8}
  n54 = Multiply(A=Saturate(None=Subtract(A=@n42, B=@n45) {constB=0}), B=Multiply(A=Multiply(A=@n46, B=@n37.A), B=Multiply(A=@n49, B=@n53)))
  n55 = DepthFade(Opacity=@n54, FadeDistance=ScalarParameter() {defaultValue=600, parameterName=DepthFade}) {fadeDistanceDefault=600}
  Multiply(A=Multiply(A=ShadingPathSwitch(Default=@n55, Forward=@n54, Mobile=@n54), B=@n41), B=ParticleColor().A)
  NamedRerouteDeclaration(None=TextureCoordinate()) {name=TexCoord, nodeColor=<rgba 1 0 0.6328 1>}
  n56 = DynamicParameter() {paramNames [Time, U_Tiling_Multiply, ErodeAll, GrowOut]}
  NamedRerouteDeclaration(None=@n56.U_Tiling_Multiply) {name=U_Tiling_Multiply, nodeColor=<rgba 1 0.155 0 1>}
  NamedRerouteDeclaration(None=@n56.Time) {name=PannerTime, nodeColor=<rgba 1 0 0.3047 1>}
  Saturate(None=Multiply(A=@n37.A) {constB=2})
 ]
}
