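# Unlit masked material for Niagara ribbons, named as the spiral ribbon effect of the GateHound emote.
# A panning ocean-noise mask brightens and slightly hue-shifts particle color, then feeds emissive adjustments.
# Scrolling cloud noise, plus a parameterized "Wind" noise layer, distorts UVs of the Cspline texture.
# That texture passes through an RGB random picker driven by a dynamic parameter, then is multiplied by particle alpha and noise.
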Material /Game/Animation/Game/MainPlayer/Emotes/GateHound/FX/Materials/M_Gatehound_Emote_SpiralRibbon
graph {
 properties {bUsedWithNiagaraRibbons=true, blendMode=BLEND_Masked, pixelDepthOffsetMode=PDOM_Legacy, shadingModel=MSM_Unlit, usedShadingModels=Unlit}
 nodes [
  n1 = TextureCoordinate()
  n2 = AppendVector(A=Constant() {r=2}, B=Constant() {r=1})
  n3 = Constant() {r=0.3}
  n4 = Constant() {r=0.02}
  n5 = TextureSample(UVs=Panner(Coordinate=Multiply(A=@n1, B=@n2), Speed=AppendVector(A=@n3, B=@n4))) {samplerType=SAMPLERTYPE_Masks, texture=/FortWater/Textures/OceanNoise__Mask}
  n6 = MaterialFunctionCall(Input=@n5.RGB, Target Low=Constant(), Target High=Constant() {r=0.01}) {materialFunction=&RemapValueRange}
  n7 = ParticleColor()
  n8 = Multiply(
   A=MaterialFunctionCall(Input=@n5.R, Target Low=Constant() {r=1}, Target High=Constant() {r=2}) {materialFunction=&RemapValueRange}.Result
   B=@n7.RGB
  )
  n9 = NamedRerouteUsage()
  MaterialFunctionCall(Emissive=Multiply(
    A=MaterialFunctionCall(Hue Shift Percentage=@n6.Result, Texture=@n8) {materialFunction=&HueShift}.Result
    B=LinearInterpolate(Alpha=@n9) {constA=0.5}
   )) {materialFunction=&MF_FX_EmissiveAdjustments}
  n10 = DynamicParameter() {paramNames [RandomPicker,Param2,Param3,Param4]}
  n11 = Constant()
  n12 = Multiply(A=TextureCoordinate(), B=AppendVector(A=Constant() {r=0.1}, B=Constant() {r=0.1}))
  n13 = AppendVector(A=Constant() {r=0.1}, B=Constant() {r=0.05})
  n14 = Multiply(
   A=TextureSample(UVs=Panner(Coordinate=@n12, Speed=@n13)) {texture=&AtmosphericCloudNoise01}.R
   B=Constant() {r=0.2}
  )
  n15 = TextureCoordinate()
  n16 = AppendVector(A=Constant() {r=0.3}, B=Constant() {r=0.3})
  n17 = Constant() {r=0.05}
  n18 = Constant() {r=0.02}
  n19 = TextureSample(UVs=Panner(Coordinate=Multiply(A=@n15, B=@n16), Speed=AppendVector(A=@n17, B=@n18))) {texture=&AtmosphericCloudNoise01}
  n20 = Constant() {r=0.1}
  n21 = Add(A=Clamp(None=AppendVector(A=@n11, B=@n14)), B=Clamp(None=Multiply(A=@n19.R, B=@n20)))
  n22 = NamedRerouteUsage()
  n23 = ScalarParameter() {defaultValue=2.5, parameterName=Wind_Power}
  n24 = Clamp(None=Multiply(A=Power(Base=@n22, Exp=@n23) {constExponent=4}, B=ScalarParameter() {defaultValue=4, parameterName=Wind_Mult}))
  n25 = ScalarParameter() {defaultValue=0.2, parameterName=Wind_Intensity}
  n26 = AppendVector(A=Add(A=ComponentMask(None=@n21) {r=true}, B=Multiply(A=@n24, B=@n25) {constB=0.2}), B=ComponentMask(None=@n21) {g=true})
  n27 = TextureSample(UVs=Add(A=TextureCoordinate(), B=Clamp(None=@n26))) {texture=/Game/Animation/Game/MainPlayer/Emotes/GateHound/FX/Textures/T_GatehoundEmote_Cspline}
  Multiply(
   A=MaterialFunctionCall(Input=@n5.RGB, Target Low=Constant(), Target High=Constant() {r=1}) {materialFunction=&RemapValueRange}.Result
   B=Multiply(
    A=@n7.A
    B=MaterialFunctionCall!(Random Picker=@n10.RandomPicker, Texture Input=@n27.RGB, Option 1=@n27.RGB, Option 2=@n27.RGB, Option 3=@n27.RGB) {materialFunction=/Game/Effects/Fort_Effects/Materials/_Functions/MF_RGB_Random_01}.Result
   )
  )
  n28 = ScalarParameter() {defaultValue=0.02, parameterName=Wind_Tile_X}
  n29 = ScalarParameter() {defaultValue=1, parameterName=Wind_Tile_Y}
  n30 = Panner(
   Coordinate=Multiply(A=TextureCoordinate(), B=AppendVector(A=@n28, B=@n29))
   Speed=AppendVector(A=ScalarParameter() {defaultValue=0.03, parameterName=Wind_Pan_X}, B=ScalarParameter() {parameterName=Wind_Pan_Y})
  )
  NamedRerouteDeclaration(None=ComponentMask(None=TextureSample(UVs=@n30) {texture=&AtmosphericCloudNoise01}.RGB) {r=true}) {name=Wind, nodeColor=<rgba 0.7344 0 1 1>}
 ]
}
