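# Unlit, two-sided translucent material for the inner vortex of the GateHound emote effect, usable on Niagara ribbons.
# Panned wispy fire noise, timed by a dynamic parameter, breaks up U gradients to blend red, orange and gold emissive color.
# Opacity multiplies a view-angle fade, U edge masks and a noise-eroded Reveal dissolve, then particle alpha.
# TwoSidedSign zeroes opacity on front faces, leaving only the inside visible.
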Material /Game/Animation/Game/MainPlayer/Emotes/GateHound/FX/Materials/M_Gatehound_Innervortex
graph {
 properties {
  bUsedWithNiagaraRibbons=true
  blendMode=BLEND_Translucent
  pixelDepthOffsetMode=PDOM_Legacy
  shadingModel=MSM_Unlit
  twoSided=true
  usedShadingModels=Unlit
 }
 nodes [
  n1 = Constant3Vector() {constant=<rgba 0.4287 0.0355 0.0312 1>}
  n2 = Constant3Vector() {constant=<rgba 0.6196 0.1439 0.0572 1>}
  n3 = Constant3Vector() {constant=<rgba 1 0.5965 0.1615 1>}
  n4 = ComponentMask(None=TextureCoordinate()) {r=true}
  n5 = SmoothStep(Value=@n4) {constMax=0, constMin=0.6}
  n6 = DynamicParameter() {defaultValue=<rgba 1 1 1 2>, paramNames [Reveal,Time,TileX,TileY]}
  n7 = MaterialFunctionCall(Float2=Multiply(A=TextureCoordinate(), B=AppendVector(A=@n6.TileX, B=@n6.TileX))) {materialFunction=&BreakOutFloat2Components}
  n8 = TextureCoordinate()
  n9 = Add(A=AppendVector(A=@n7.R, B=@n7.R), B=Multiply(A=ComponentMask(None=@n8) {r=true}) {constB=2})
  n10 = DynamicParameter() {paramNames [Reveal,Time,Param3,Param4]}
  n11 = Saturate(None=TextureSample(UVs=Panner(Coordinate=@n9, Time=@n10.Time) {speedX=2, speedY=-1}) {samplerType=SAMPLERTYPE_Grayscale, texture=/Game/Effects/Fort_Effects/Textures/Fire/T_WispyFireNoise_Soft_M}.R)
  n12 = Saturate(None=Subtract(A=Multiply(A=@n5) {constB=2}, B=@n11))
  n13 = Power(Base=@n11) {constExponent=0.5}
  n14 = SmoothStep(Value=@n4) {constMax=0, constMin=1}
  MaterialFunctionCall(Emissive=LinearInterpolate(
    A=MaterialFunctionCall(A=@n1, B=@n2, C=@n3, Alpha=@n12) {materialFunction=&3ColorBlend}.Result
    B=Constant3Vector() {constant=<rgba 0.2604 0.0692 0 1>}
    Alpha=Multiply(A=@n13, B=@n14)
   )) {materialFunction=&MF_FX_EmissiveAdjustments}
  n15 = DotProduct(A=PixelNormalWS(), B=CameraVectorWS())
  n16 = Multiply(
   A=Power(Base=SmoothStep(Value=@n15) {constMax=0.5}, Exp=ScalarParameter() {defaultValue=0.25, parameterName=Fadepower})
   B=ScalarParameter() {defaultValue=2, parameterName=Fadepower Mask Strenght}
  )
  n17 = Multiply(A=SmoothStep(Value=@n4) {constMax=0.2}, B=@n14)
  n18 = SmoothStep(Min=DynamicParameter() {paramNames [Reveal,Time,Param3,Param4]}.Reveal, Value=@n4) {constMax=0, constMin=1}
  n19 = Saturate(None=Subtract(A=Multiply(A=@n18) {constB=1.5}, B=Multiply(A=@n11) {constB=0.5}))
  n20 = Saturate(None=Multiply(A=@n16, B=Multiply(A=@n17, B=@n19)))
  n21 = TwoSidedSign()
  Multiply(A=Saturate(None=LinearInterpolate(A=@n20, Alpha=@n21) {constB=0}), B=ParticleColor().A)
  Saturate(None=Subtract(A=@n16, B=@n11))
 ]
}
